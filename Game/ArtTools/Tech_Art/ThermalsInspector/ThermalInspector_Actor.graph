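# ThermalInspector_Actor is an Editor Utility Blueprint actor with a point light, apparently a helper for the ThermalsInspector tech-art tool.
# Its setMultiplier function multiplies the stored defaultIntensity by the given multiplier and applies the result as the PointLight intensity.
# The event graph and construction script contain no logic, so the light changes only when setMultiplier is called.
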EditorUtilityBlueprint /Game/ArtTools/Tech_Art/ThermalsInspector/ThermalInspector_Actor
graphs [
 EventGraph {nodes [
   Event BeginPlay()
   Event ActorBeginOverlap()
   Event Tick!() -> (OutputDelegate, then, DeltaSeconds:float)
  ]}
 UserConstructionScript {nodes [Construction Script()]}
 setMultiplier {nodes [
   n1 = setMultiplier!() -> (then, mult:double=~) {is=title K2Node_FunctionEntry}
   > SetIntensity!(
    execute
    self:Light Component Object Reference=Get PointLight() -> (PointLight:Point Light Component Object Reference)
    NewIntensity:float=float * float(A=@n1, B=Get defaultIntensity() -> (defaultIntensity:double)) {is=title K2Node_PromotableOperator}
   ) -> (then)
  ]}
]
variables [defaultIntensity {type=number}]
