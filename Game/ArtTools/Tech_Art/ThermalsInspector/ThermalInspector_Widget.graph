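# Editor Utility Widget with an "Inspect Thermals" toggle button that turns a thermal inspection view on and off in the editor.
# Clicking spawns a transient, tagged ThermalInspector_Actor at the origin or destroys it, tracking bActive and restyling the button blue or grey.
# On Construct it searches the editor world by tag to re-adopt an existing actor and show the active state; on Destruct it destroys the actor.
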EditorUtilityWidgetBlueprint /Game/ArtTools/Tech_Art/ThermalsInspector/ThermalInspector_Widget
graphs [
 DestroyThermalInspector {nodes [
   DestroyThermalInspector() {is=title K2Node_FunctionEntry}
   n1 = Get ThermalInspectorPostProcess() -> (ThermalInspectorPostProcess:Thermal Inspector Actor Object Reference)
   > Is Valid(InputObject=@n1)
   > DestroyActor!(execute, self:Editor Actor Subsystem Object Reference=Get EditorActorSubsystem(), ActorToDestroy:Actor Object Reference=@n1) -> (then, ReturnValue:bool)
  ]}
 EventGraph {nodes [
   "On Clicked (Button_32)"() {is=title K2Node_ComponentBoundEvent}
   > n1 = Branch(Condition=Get bActive() -> (bActive:bool))
   > SetButtonInactive()
   > DestroyThermalInspector() {is=title K2Node_CallFunction}
   SetButtonActive(execute=@n1.else)
   > SpawnThermalInspector()
   Event Destruct()
   > DestroyThermalInspector()
   Event Construct()
   > FindThermalPostProcessByTag()
   Event Tick()
  ]}
 FindThermalPostProcessByTag {nodes [
   FindThermalPostProcessByTag() {is=title K2Node_FunctionEntry}
   > n1 = GetEditorWorld(self=Get UnrealEditorSubsystem())
   > n2 = GetAllActorsOfClassWithTag!(
    execute
    WorldContextObject:Object Reference=@n1
    ActorClass:Actor Class Reference=&/ThermalInspector_Actor.ThermalInspector_Actor_C
    Tag:Name=ThermalInspector
   ) -> (then, OutActors:Array of Thermal Inspector Actor Object References)
   n3 = "Get (a copy)"(Array:Array of Thermal Inspector Actor Object References=@n2) -> (Output:Thermal Inspector Actor Object Reference)
   > Is Valid(InputObject=@n3)
   > Set ThermalInspectorPostProcess(ThermalInspectorPostProcess:Thermal Inspector Actor Object Reference=@n3) -> (Output_Get:Thermal Inspector Actor Object Reference)
   > SetButtonActive()
  ]}
 SetButtonActive {nodes [
   SetButtonActive() {is=title K2Node_FunctionEntry}
   n1 = Make Slate Color!(SpecifiedColor=(R=0.0395,G=0.0395,B=0.0395,A=1)) -> (SlateColor)
   n2 = Make Slate Color(SpecifiedColor=(R=0,G=0.162,B=0.7454,A=1))
   n3 = Make Slate Color(SpecifiedColor=(R=0.0044,G=0.2384,B=1,A=1))
   n4 = Make Slate Color(SpecifiedColor=(R=0.0395,G=0.0395,B=0.0395,A=1))
   > SetStyle!(
    execute
    self:Button Object Reference=Get Button_32() -> (Button_32:Button Object Reference)
    InStyle:"Button Style Structure (by ref)"=Make Button Style!(
     Normal:Slate Brush Structure=Make Slate Brush!(TintColor:Slate Color Structure=@n2) -> (SlateBrush:Slate Brush Structure) {is=title K2Node_MakeStruct}
     Hovered:Slate Brush Structure=Make Slate Brush(TintColor=@n3)
     Pressed:Slate Brush Structure=Make Slate Brush(TintColor=@n4)
     Disabled:Slate Brush Structure=Make Slate Brush(TintColor=@n1)
     ClickedSlateSound:Slate Sound Structure=(ResourceObject=None)
    ) -> (ButtonStyle:Button Style Structure) {is=title K2Node_MakeStruct}
   ) -> (then)
   > Set bActive(bActive:bool=true) -> (Output_Get:bool)
  ]}
 SetButtonInactive {nodes [
   SetButtonInactive() {is=title K2Node_FunctionEntry}
   n1 = Make Slate Color(SpecifiedColor=(R=0.0395,G=0.0395,B=0.0395,A=1))
   n2 = Make Slate Color(SpecifiedColor=(R=0.0395,G=0.0395,B=0.0395,A=1))
   n3 = Make Slate Color(SpecifiedColor=(R=0.0953,G=0.0953,B=0.0953,A=1))
   n4 = Make Slate Color(SpecifiedColor=(R=0.0284,G=0.0284,B=0.0284,A=1))
   > SetStyle(
    self=Get Button_32()
    InStyle=Make Button Style(
     Normal=Make Slate Brush(TintColor=@n2)
     Hovered=Make Slate Brush(TintColor=@n3)
     Pressed=Make Slate Brush(TintColor=@n4)
     Disabled=Make Slate Brush(TintColor=@n1)
     ClickedSlateSound=(ResourceObject=None)
    )
   )
   > Set bActive(bActive=false)
  ]}
 SpawnThermalInspector {nodes [
   SpawnThermalInspector() {is=title K2Node_FunctionEntry}
   > n1 = SpawnActorFromClass!(
    execute
    self:Editor Actor Subsystem Object Reference=Get EditorActorSubsystem()
    ActorClass:Actor Class Reference=&ThermalInspector_Actor_C
    Location:Vector="0, 0, 0"
    Rotation:Rotator
    bTransient:bool=true
   ) -> (then, ReturnValue:Thermal Inspector Actor Object Reference)
   > n2 = Set ThermalInspectorPostProcess(ThermalInspectorPostProcess=@n1)
   > Set Tags!(execute, Tags:Array of Names=Get Tags() -> (Tags:Array of Names), self:Actor Object Reference=@n2) -> (then, Output_Get:Array of Names)
  ]}
]
variables [
 bActive {type=boolean}
 Tags {type="array<string>", default [ThermalInspector]}
 ThermalInspectorPostProcess {type=ThermalInspector_Actor_C, default=null}
]
widgets {
 info {rootWidgetClass=&CanvasPanel, widgetCount=3}
 widgets [
  CanvasPanel_0 {widgetClassPath=&CanvasPanel, widget=CanvasPanel_0}
  Button_32 {
   widgetClassPath=&Button
   parent=CanvasPanel_0
   widget=Button_32
   properties {widgetStyle {
     clickedSlateSound {resourceObject=null}
     disabled {
      bIsDynamicallyLoaded=false
      drawAs=RoundedBox
      imageType=NoImage
      mirroring=NoMirror
      resourceName=None
      resourceObject=null
      tiling=NoTile
      imageSize=<0 0>
      margin=<ltrb 0 0 0 0>
      outlineSettings {
       bUseBrushTransparency=false
       roundingType=FixedRadius
       width=1
       color {colorUseRule=UseColor_ColorTable, specifiedColor=<rgba 0 0 0 0>}
       cornerRadii=<4 4 4 4>
      }
      tintColor {colorUseRule=UseColor_Specified, specifiedColor=<rgba 0.0395 0.0395 0.0395 1>}
      uVRegion {bIsValid=false, max=<0 0>, min=<0 0>}
     }
     disabledForeground {colorUseRule=UseColor_ColorTable, specifiedColor=<rgba 0 0 0 0>}
     hovered {
      bIsDynamicallyLoaded=false
      drawAs=RoundedBox
      imageType=NoImage
      mirroring=NoMirror
      resourceName=None
      resourceObject=null
      tiling=NoTile
      imageSize=<32 32>
      margin=<ltrb 0 0 0 0>
      outlineSettings {
       bUseBrushTransparency=false
       roundingType=FixedRadius
       width=1
       color {colorUseRule=UseColor_ColorTable, specifiedColor=<rgba 0 0 0 0>}
       cornerRadii=<4 4 4 4>
      }
      tintColor {colorUseRule=UseColor_Specified, specifiedColor=<rgba 0.0953 0.0953 0.0953 1>}
      uVRegion {bIsValid=false, max=<0 0>, min=<0 0>}
     }
     hoveredForeground {colorUseRule=UseColor_ColorTable, specifiedColor=<rgba 0 0 0 0>}
     hoveredSlateSound {resourceObject=null}
     normal {
      bIsDynamicallyLoaded=false
      drawAs=RoundedBox
      imageType=NoImage
      mirroring=NoMirror
      resourceName=None
      resourceObject=null
      tiling=NoTile
      imageSize=<32 32>
      margin=<ltrb 0 0 0 0>
      outlineSettings {
       bUseBrushTransparency=false
       roundingType=FixedRadius
       width=1
       color {colorUseRule=UseColor_ColorTable, specifiedColor=<rgba 0 0 0 0>}
       cornerRadii=<4 4 4 4>
      }
      tintColor {colorUseRule=UseColor_Specified, specifiedColor=<rgba 0.0395 0.0395 0.0395 1>}
      uVRegion {bIsValid=false, max=<0 0>, min=<0 0>}
     }
     normalForeground {colorUseRule=UseColor_ColorTable, specifiedColor=<rgba 0 0 0 0>}
     normalPadding=<ltrb 12 1.5 12 1.5>
     pressed {
      bIsDynamicallyLoaded=false
      drawAs=RoundedBox
      imageType=NoImage
      mirroring=NoMirror
      resourceName=None
      resourceObject=null
      tiling=NoTile
      imageSize=<32 32>
      margin=<ltrb 0 0 0 0>
      outlineSettings {
       bUseBrushTransparency=false
       roundingType=FixedRadius
       width=1
       color {colorUseRule=UseColor_ColorTable, specifiedColor=<rgba 0 0 0 0>}
       cornerRadii=<4 4 4 4>
      }
      tintColor {colorUseRule=UseColor_Specified, specifiedColor=<rgba 0.0284 0.0284 0.0284 1>}
      uVRegion {bIsValid=false, max=<0 0>, min=<0 0>}
     }
     pressedForeground {colorUseRule=UseColor_ColorTable, specifiedColor=<rgba 0 0 0 0>}
     pressedPadding=<ltrb 12 2.5 12 0.5>
     pressedSlateSound {resourceObject=null}
    }}
   slot {
    class=CanvasPanelSlot
    properties {layoutData {alignment=<0.5 0.5>, anchors {maximum=<0.5 0.5>, minimum=<0.5 0.5>}, offsets=<ltrb 0 0 140 30>}}
   }
  }
  UEFN_TextBlock_C_93 {
   widgetClassPath=&UEFN_TextBlock
   parent=Button_32
   widget=UEFN_TextBlock_C_93
   properties {
    text=Inspect Thermals
    colorAndOpacity {colorUseRule=UseColor_Specified, specifiedColor=<rgba 0.5 0.5 0.5 1>}
    font {
     bForceMonospaced=false
     bMaterialIsStencil=false
     fontMaterial=null
     fontObject=&Roboto
     letterSpacing=0
     monospacedWidth=1
     size=9
     skewAmount=0
     typefaceFontName=Regular
     outlineSettings {
      bApplyOutlineToDropShadows=false
      bMiteredCorners=false
      bSeparateFillAlpha=false
      outlineMaterial=null
      outlineSize=0
      outlineColor=<rgba 0 0 0 1>
     }
    }
   }
   slot {class=ButtonSlot}
  }
 ]
}
